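# Material function that adds a frosting look to the Birthday 5 backpack, blending base color and metallic within a selectable SkinFX mask channel.
# Frosting color comes from a gradient texture sampled with inverted Fresnel plus tiled screen position, making it view-dependent.
# An optional panning swirl normal map's red channel distorts that lookup; its Speed_X drives both pan axes.
# ThinFilm and HueShift calls are present but unconnected.
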MaterialFunction /Game/Accessories/FORT_Backpacks/Backpack_M_MED_FN_Birthday_5/Materials/MF_Birthday5_Frosting
graph {nodes [
  n1 = LinearInterpolate(
   A=GetMaterialAttributes(None=NamedRerouteUsage()) {attributeGetTypes [#1]}.BaseColor
   B=NamedRerouteUsage()
   Alpha=NamedRerouteUsage()
  )
  n2 = LinearInterpolate(
   A=GetMaterialAttributes(None=NamedRerouteUsage()) {attributeGetTypes [#2]}.Metallic
   B=ScalarParameter() {defaultValue=0.5, group=Frosting, parameterName=Metalness, sliderMax=1}
   Alpha=NamedRerouteUsage()
  )
  FunctionOutput(None=SetMaterialAttributes!(
    MaterialAttributes=NamedRerouteUsage()
    BaseColor=NamedRerouteDeclaration(None=@n1) {name=MA_Base_Color, nodeColor=<rgba 0 0.3438 1 1>}
    Metallic=NamedRerouteDeclaration(None=@n2) {name=MA_Metallic, nodeColor=<rgba 0.4062 0 1 1>}
   ) {attributeSetTypes [#1, #2]})
  MaterialFunctionCall!(Alpha, Material Attributes) {materialFunction=/Game/Weapons/WeaponSkins/MaterialFunction/ThinFilmFunction}
  MaterialFunctionCall() {materialFunction=&HueShift}
  NamedRerouteDeclaration(None=FunctionInput() {inputName=Material Attributes, inputType=FunctionInput_MaterialAttributes}) {name=Material_Attributes, nodeColor=<rgba 0 1 0.8594 1>}
  n3 = OneMinus(None=Fresnel(
    ExponentIn=ScalarParameter() {defaultValue=0.3, group=Frosting, parameterName=ExponentIn}
    BaseReflectFractionIn=ScalarParameter() {defaultValue=0.04, group=Frosting, parameterName=BaseReflectFractionIn}
   ) {exponent=0.3})
  n4 = ScreenPosition()
  n5 = Add(
   A=AppendVector(A=@n3, B=@n3)
   B=Multiply(A=@n4.ViewportUV, B=ScalarParameter() {defaultValue=0.2, group=Frosting, parameterName=Frosting_Tiling})
  )
  n6 = StaticSwitchParameter(True=TextureCoordinate(), False=@n4.ViewportUV) {defaultValue=true, group=Frosting, parameterName=Swirl_Use_UV}
  n7 = ScalarParameter() {defaultValue=1, group=Frosting, parameterName=Swirl_Tiling}
  n8 = VectorParameter() {group=Frosting, parameterName=Swirl_Speed, channelNames {a="", b="", g=Speed_Y, r=Speed_X}}
  n9 = TextureSampleParameter2D(UVs=Panner(Coordinate=Multiply(A=@n6, B=@n7), Speed=AppendVector(A=@n8.Speed_X, B=@n8.Speed_X))) {
   group=Frosting
   parameterName=SwirlNormal
   samplerType=SAMPLERTYPE_Normal
   texture=/Game/Accessories/FORT_Backpacks/Backpack_M_MED_FN_Birthday_5/Textures/TechArt/T_M_MED_FN_Birthday_5_Backpack_Swirl_Normal
  }
  n10 = Multiply(A=AppendVector(A=@n9.R, B=@n9.R), B=ScalarParameter() {defaultValue=1, group=Frosting, parameterName=Swirl_Strength})
  n11 = TextureSampleParameter2D(UVs=StaticSwitchParameter(True=Add(A=@n5, B=@n10), False=@n5) {group=Frosting, parameterName=Swirl_Enable}) {
   group=Frosting
   parameterName=Frosting_Gradient
   samplerType=SAMPLERTYPE_LinearColor
   texture=/Game/Accessories/FORT_Backpacks/Backpack_M_MED_FN_Birthday_5/Textures/TechArt/CA_M_MED_FN_Birthday_5_Backpack_Frosting_Gradient
  }
  NamedRerouteDeclaration(None=@n11.RGB) {name=Frosting, nodeColor=<rgba 1 0.2578 0 1>}
  NamedRerouteDeclaration(None=ChannelMaskParameter(None=FunctionInput() {inputName=SkinFX_Mask, inputType=FunctionInput_Vector4}) {group=Frosting, parameterName=Frosting_Mask_Channel}.R) {name=Frosting_Mask, nodeColor=<rgba 0 1 0.5781 1>}
  TextureCoordinate()
 ]}
