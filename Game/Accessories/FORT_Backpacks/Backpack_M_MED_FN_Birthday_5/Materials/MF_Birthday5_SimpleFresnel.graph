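# Material function for the Birthday 5 backpack that applies a Fresnel-driven color tint to the base color of incoming material attributes.
# An input Fresnel value is raised to a power parameter, multiplied by a color parameter, then multiplied into the base color rather than added.
# A selectable channel of the Mask input blends original and tinted color.
# A static switch toggles the effect, writing only base color back into the attributes.
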MaterialFunction /Game/Accessories/FORT_Backpacks/Backpack_M_MED_FN_Birthday_5/Materials/MF_Birthday5_SimpleFresnel
graph {nodes [
  n1 = GetMaterialAttributes(None=FunctionInput() {inputType=FunctionInput_MaterialAttributes}) {attributeGetTypes [#1]}
  n2 = Power(
   Base=FunctionInput() {inputName=Fresnel, inputType=FunctionInput_Scalar}
   Exp=ScalarParameter() {defaultValue=1, group=Birthday5_SimpleFresnel, parameterName=Birthday5_Fresnel_Power}
  )
  n3 = VectorParameter() {group=Birthday5_SimpleFresnel, parameterName=Color, defaultValue=<rgba 0.9323 0.7111 0.3105 0>}
  n4 = LinearInterpolate(
   A=@n1.BaseColor
   B=Multiply(A=@n1.BaseColor, B=Multiply(A=@n2, B=@n3.RGB))
   Alpha=ChannelMaskParameter(None=FunctionInput() {inputName=Mask, inputType=FunctionInput_Vector4}) {group=Birthday5_SimpleFresnel, parameterName=Mask_Channel}.R
  )
  FunctionOutput(None=SetMaterialAttributes!(
    MaterialAttributes=@n1.MaterialAttributes
    BaseColor=StaticSwitchParameter(True=@n4, False=@n1.BaseColor) {group=Birthday5_SimpleFresnel, parameterName=Use Fresnel}
   ) {attributeSetTypes [#1]})
 ]}
