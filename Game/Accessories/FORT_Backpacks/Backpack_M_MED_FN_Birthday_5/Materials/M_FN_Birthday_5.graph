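# Lit character material for the fifth Fortnite Birthday backpack, layering birthday-themed effects onto the shared character base on skeletal meshes.
# Base PreFX material attributes pass through an optional metals LUT switch, then Birthday 5 fresnel, glitch confetti and frosting functions.
# A SkinFX mask texture drives the effects, while an opacity mask and pre-skinned local position also feed the confetti.
# The result finishes through character PostFX, cinematic PostFX and game-wide shared functions.
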Material /Game/Accessories/FORT_Backpacks/Backpack_M_MED_FN_Birthday_5/Materials/M_FN_Birthday_5
graph {
 properties {
  bUseMaterialAttributes=true
  bUsedWithMeshDeformer=true
  bUsedWithSkeletalMesh=true
  pixelDepthOffsetMode=PDOM_Legacy
  usedShadingModels=Default Lit
 }
 nodes [
  n1 = MaterialFunctionCall() {materialFunction=&BaseFN_CharacterFunction_PreFX}
  n2 = NamedRerouteUsage()
  n3 = MaterialFunctionCall!(
   In=StaticSwitchParameter(
    True=MaterialFunctionCall(Mask=@n2, Materials=@n1.Result) {materialFunction=&MF_MetalsLUT}.Result
    False=@n1.Result
   ) {parameterName=Use_MetalLut}
   Mask=@n2
   Fresnel=@n1.Fresnel
  ) {materialFunction=&/MF_Birthday5_SimpleFresnel}
  n4 = NamedRerouteUsage()
  n5 = NamedRerouteUsage()
  n6 = MaterialFunctionCall(
   AO=@n1.AO
   Distance Blend=@n1.AO
   Fresnel=@n1.AO
   Gmap=@n1.AO
   Pre Skinned Local Position=NamedRerouteDeclaration(None=@n1.Pre skinned Local Position) {name=Pre-Skinned Local Position, nodeColor=<rgba 0 0.7422 1 1>}
   Material=MaterialFunctionCall!(
    SkinFX_Mask=NamedRerouteUsage()
    Material Attributes=MaterialFunctionCall!(SkinFX_Mask=@n2, Material Attributes=@n3.Result, Opacity_Mask=@n4, Pre-Skinned Local Position=@n5) {materialFunction=&/MF_Birthday5_GlitchConfetti}.Material_Attributes
   ) {materialFunction=&/MF_Birthday5_Frosting}.Result
  ) {materialFunction=&BaseFN_CharacterFunction_PostFX}
  MaterialFunctionCall!(In=MaterialFunctionCall(In=@n6.Result) {materialFunction=&CINE_MF_CharPostFX}.Result) {materialFunction=/Game/Athena/WinterArt/Materials/MF_GameWideShared}
  NamedRerouteDeclaration(None=TextureSample(Tex=TextureObjectParameter() {group=Skin FX, parameterName=SkinFX_Mask}).RGBA) {name=SkinFX_Mask, nodeColor=<rgba 0.9842 0 1 1>}
  NamedRerouteDeclaration(None=TextureSampleParameter2D() {group=Confetti, parameterName=OpacityMask}.RGBA) {name=OpacityMask, nodeColor=<rgba 0.5938 1 0 1>}
 ]
}
