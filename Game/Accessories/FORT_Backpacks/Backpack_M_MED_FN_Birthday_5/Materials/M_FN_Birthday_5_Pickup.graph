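# Subsurface pickup material for the Birthday 5 backpack, adding an animated shine to the birthday look.
# Character PreFX attributes pass through masked glitch confetti, optional metal LUT, frosting and simple fresnel functions.
# Emissive is then set to a switchable panning sweep band, projected from pivot-relative world position along a camera-space direction, plus optional powerup fresnel.
# Curie fallback, creative move tool and character PostFX functions are also included.
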Material /Game/Accessories/FORT_Backpacks/Backpack_M_MED_FN_Birthday_5/Materials/M_FN_Birthday_5_Pickup
graph {
 properties {
  bUseMaterialAttributes=true
  bUsedWithMeshDeformer=true
  bUsedWithMeshParticles=true
  bUsedWithSkeletalMesh=true
  pixelDepthOffsetMode=PDOM_Legacy
  shadingModel=MSM_Subsurface
  usedShadingModels=Subsurface
 }
 nodes [
  MaterialFunctionCall() {materialFunction=&BaseFN_CharacterFunction_PostFX}
  MaterialFunctionCall(MA In=MaterialFunctionCall!(In=NamedRerouteUsage(), UseSphericalMappingOverride) {materialFunction=/Game/Packages/Fortress_SharedMaterials/Base_Material_Functions/MF_CurieFallbackEffect}.Result) {materialFunction=&MF_CreativeMoveTool}
  n1 = NamedRerouteUsage()
  n2 = MaterialFunctionCall() {materialFunction=&BaseFN_CharacterFunction_PreFX}
  n3 = MaterialFunctionCall!(
   SkinFX_Mask=@n1
   Material Attributes=@n2.Result
   Opacity_Mask=Constant()
   Pre-Skinned Local Position=NamedRerouteDeclaration(None=@n2.Pre skinned Local Position) {name=Pre-Skinned Local Position, nodeColor=<rgba 0 0.7422 1 1>}
  ) {materialFunction=&/MF_Birthday5_GlitchConfetti}
  n4 = MaterialFunctionCall!(
   SkinFX_Mask=@n1
   Material Attributes=StaticSwitchParameter(
    True=MaterialFunctionCall(Mask=@n1, Materials=@n3.Material_Attributes) {materialFunction=&MF_MetalsLUT}.Result
    False=@n3.Material_Attributes
   ) {parameterName=Use_MetalLut}
  ) {materialFunction=&/MF_Birthday5_Frosting}
  n5 = ScalarParameter() {defaultValue=1, parameterName=PowerupFresnelBrightness}
  n6 = ScalarParameter() {defaultValue=1, parameterName=PowerupFresnelExponent}
  n7 = Multiply(
   A=Subtract(
    A=WorldPosition().XYZ
    B=MaterialFunctionCall() {materialFunction=&ComponentPivotLocation}.Result
   )
   B=ScalarParameter() {defaultValue=0.004, parameterName=WorldTileScale}
  )
  n8 = Normalize(VectorInput=Transform(None=VectorParameter() {parameterName=SweepDirectionCamSpace, defaultValue=<rgba -0.5 0.45 -1 0>}.RGB) {transformSourceType=TRANSFORMSOURCE_Camera})
  n9 = ComponentMask(None=Multiply(A=DotProduct(A=@n7, B=@n8), B=@n8)) {b=true, r=true}
  n10 = ScalarParameter() {defaultValue=0.18, parameterName=PanSpeed}
  n11 = Frac(None=ComponentMask(None=Panner(Coordinate=@n9, Speed=@n10) {speedX=0.2, speedY=0.2}) {g=true})
  n12 = Multiply(A=Subtract(A=@n11) {constB=0.97}) {constB=33.3333}
  n13 = Multiply(A=Floor(None=Add(A=@n11) {constB=0.03}), B=Power(Base=OneMinus(None=@n12)) {constExponent=3})
  n14 = VectorParameter() {parameterName=SweepColour, defaultValue=<rgba 2 1.6277 0.6 0>}
  n15 = Add(A=StaticSwitchParameter(True=Multiply(A=@n13, B=@n14.RGB), False=Constant()) {defaultValue=true, parameterName=UseSweep}) {constB=0}
  n16 = Add(
   A=Multiply(
    A=VectorParameter() {parameterName=PowerupFresnelColor, defaultValue=<rgba 1 1 1 0>}.RGB
    B=MaterialFunctionCall(Brightness=@n5, Exponent=@n6) {materialFunction=&MF_Fresnel_FX}.Fresnel Inverted
   )
   B=@n15
  )
  NamedRerouteDeclaration(None=SetMaterialAttributes(
    MaterialAttributes=MaterialFunctionCall!(In=@n4.Result, Mask=@n1, Fresnel=@n2.Fresnel) {materialFunction=&/MF_Birthday5_SimpleFresnel}.Result
    Emissive Color=StaticSwitchParameter(True=@n16, False=@n15) {parameterName=CreativePowerupFresnel}
   ) {attributeSetTypes [#1]}) {name=DurrBurgerPickup_MA, nodeColor=<rgba 0.875 1 0 1>}
  NamedRerouteDeclaration(None=TextureSample(Tex=TextureObjectParameter() {group=Skin FX, parameterName=SkinFX_Mask}).RGBA) {name=SkinFX_Mask, nodeColor=<rgba 0.9842 0 1 1>}
 ]
}
