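# Material function for the Birthday 5 backpack that adds an optional colored rim outline to a material's emissive.
# An inverted Fresnel through a step gives a hard-edged band, masked by a selectable opacity-mask channel and scaled by outline color and brightness.
# A static switch blends it into emissive using that same mask, then writes it back into the material attributes.
# A Power node off the Fresnel is left unconnected to the output.
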MaterialFunction /Game/Accessories/FORT_Backpacks/Backpack_M_MED_FN_Birthday_5/Materials/MF_Birthday5_Outline
graph {nodes [
  n1 = FunctionInput() {inputName=Material Attributes, inputType=FunctionInput_MaterialAttributes}
  n2 = GetMaterialAttributes(None=@n1) {attributeGetTypes [#1]}
  n3 = ChannelMaskParameter(None=FunctionInput() {inputName=Opacity_Mask, inputType=FunctionInput_Vector4}) {desc=Reads from Opacity mask, group=Outline, parameterName=Outline_Mask_Channel}
  n4 = StaticSwitchParameter(True=LinearInterpolate(A=@n2.EmissiveColor, B=NamedRerouteUsage(), Alpha=@n3.R), False=@n2.EmissiveColor) {group=Outline, parameterName=Enable_Outline}
  FunctionOutput(None=SetMaterialAttributes!(MaterialAttributes=@n1, EmissiveColor=NamedRerouteDeclaration(None=@n4) {name=MA_Emissive, nodeColor=<rgba 0 1 0.8359 1>}) {attributeSetTypes [#1]})
  n5 = Fresnel(
   ExponentIn=ScalarParameter() {defaultValue=1, group=Outline, parameterName=Outline_ExponentIn}
   BaseReflectFractionIn=ScalarParameter() {defaultValue=0.2, group=Outline, parameterName=Outline_BaseReflectFractionIn}
  )
  n6 = Multiply(A=Step(Y=OneMinus(None=@n5)) {constX=0.5}, B=@n3.R)
  n7 = VectorParameter() {group=Outline, parameterName=Outline_Color, defaultValue=<rgba 0.1908 0.037 0.474 1>}
  NamedRerouteDeclaration(None=Multiply(A=Multiply(A=@n6, B=@n7.RGB), B=ScalarParameter() {defaultValue=1, group=Outline, parameterName=Outline_Brightness})) {name=OutlIne, nodeColor=<rgba 0 0.3203 1 1>}
  Power(Base=@n5) {constExponent=25}
 ]}
