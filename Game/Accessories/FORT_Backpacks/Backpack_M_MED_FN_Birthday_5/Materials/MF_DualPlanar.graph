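# Dual-planar UV material function for the Birthday 5 backpack, outputting tiled coordinates projected from pre-skinned local position.
# The YZ projection pans by the Speed input, while the XY projection takes a fixed rotation without panning.
# Absolute pre-skinned normals, passed through a vertex interpolator, give a hard step mask selecting XY where Z dominates X; normal Y is ignored.
# The selected coordinates are multiplied by Tiling after panning and rotation.
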MaterialFunction /Game/Accessories/FORT_Backpacks/Backpack_M_MED_FN_Birthday_5/Materials/MF_DualPlanar
graph {nodes [
  n1 = FunctionInput() {inputName=Pre-Skinned Local Position}
  n2 = FunctionInput() {bUsePreviewValueAsDefault=true, inputName=Speed, inputType=FunctionInput_Vector2, previewValue=<1 0 0 1>}
  n3 = Panner(Coordinate=ComponentMask(None=@n1) {b=true, g=true}, Speed=@n2) {speedY=1}
  n4 = MaterialFunctionCall(UVs=ComponentMask(None=@n1) {g=true, r=true}, "Rotation Angle (0-1)"=Constant() {r=0.355}) {materialFunction=&CustomRotator}
  n5 = Abs(None=VertexInterpolator(VS=PreSkinnedNormal()).PS)
  n6 = MaterialFunctionCall(Float3=@n5) {materialFunction=&BreakOutFloat3Components}
  n7 = Max(A=@n6.R, B=@n6.R)
  n8 = MaterialFunctionCall(Float3=Step(Y=Max(A=@n7, B=@n6.B), X=@n5)) {materialFunction=&BreakOutFloat3Components}
  n9 = FunctionInput() {bUsePreviewValueAsDefault=true, inputName=Tiling, inputType=FunctionInput_Scalar, previewValue=<5 0 0 1>}
  FunctionOutput(None=Multiply(A=LinearInterpolate(A=@n3, B=@n4.Rotated Values, Alpha=@n8.B) {constAlpha=1}, B=@n9) {constB=0.2})
 ]}
